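# Sound cue for a looping screwdriver interior room tone in the Artemis metal room-tone set.
# A looping wave player feeds an Enveloper node, whose volume curve gives the loop a short fade-in at the start.
# It uses the Athena indoor ambient sound class and shared ambient attenuation settings, with the overall volume lowered.
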SoundCue /Game/Sounds/Artemis/RoomTones/Metal/Artemis_Interior_Screwdriver_Loop_Cue
root=Enveloper_0
nodes [
 {
  id=Enveloper_0
  children [WavePlayer_0]
  properties {pitchCurve {editorCurveData {keys [{time=0}]}}, volumeCurve {editorCurveData {keys [<at 0 0 RCIM_Cubic>,<at 0.25 1>]}}}
 }
 {id=WavePlayer_0, properties {bLooping=true, soundWaveAssetPtr=&/Artemis_Interior_Screwdriver_Loop}}
]
properties {
 attenuationSettings=&Fort_Ambient_Athena_Att
 maxDistance=2097152
 soundClassObject=&Fort_Ambient_Indoors_Athena
 volumeMultiplier=0.652
}
